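# Physics asset for the Sailor Squad Rose female two-handed pickaxe, providing bodies for its handle and dynamic wire bones.
# A capsule covers the handle, with a small box on the wire master bone and boxes on the left and right wire bones.
# The master swings and twists freely relative to the handle, while both side wires constrain to the master with matching, uneven angular limits, narrow on one swing axis and twist.
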PhysicsAsset /BRCosmetics/Weapons/FORT_Melee/Pickaxe_Sailor_Squad_Rose_Female/Meshes/Sailor_Squad_Rose_Female_Physics
bodies {
 Melee_TwoHanded_Handle [{shapeType=Capsule, length=92.5942, radius=9.5466, center=<0 0 21.4423>}]
 dyn_wire_l_01 [{shapeType=Box, extentX=6.92, extentY=9.1322, extentZ=2, center=<4.4443 0 -0.0001>}]
 dyn_wire_master [{shapeType=Box, extentX=2, extentY=2, extentZ=2}]
 dyn_wire_r_01 [{shapeType=Box, extentX=6.92, extentY=9.1322, extentZ=2, center=<4.7956 0 -0.0001>}]
}
constraints [
 {
  bone1Name=dyn_wire_master
  bone2Name=Melee_TwoHanded_Handle
  swing1LimitDegrees=45
  swing1Motion=Free
  swing2LimitDegrees=45
  swing2Motion=Free
  twistLimitDegrees=45
  twistMotion=Free
 }
 {bone1Name=dyn_wire_l_01, bone2Name=dyn_wire_master, swing1LimitDegrees=10, swing2LimitDegrees=45, twistLimitDegrees=8}
 {bone1Name=dyn_wire_r_01, bone2Name=dyn_wire_master, swing1LimitDegrees=10, swing2LimitDegrees=45, twistLimitDegrees=8}
]
